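; Galileo Bun automatic weapon Blueprint that drives overheat particles, audio and force feedback.
; While heating, a looping timer, slower on mobile, converts heat to a byte that drives the muzzle heat emitter; the replicated path is left unhooked, noted as disabled for perf.
; Overheating swaps to a second particle effect with release sound and rumble; equipping and finished cooling deactivate the heat effect.
; Locally controlled firing plays an audio layer with curve-based volume from heat.
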(event FortWeapon|EventOnEquippedAndAttached
  ;[1 1 1] On Equip -> deactivate FX or else they get stuck on if previously heated up
  (|CallParentFunction)
  (Components|Activation|Deactivate (Variables|Default|GetP_Overheat)))

(event Overheating|EventOnStartOverheated
  ;[1 1 1] Overheated
  (|CallParentFunction)
  (Components|Activation|Activate (Variables|Default|GetP_Overheat_V2) true)
  (Audio|SpawnSoundAttached "/Game/Sounds/Fort_Weapons/AutoRifle/Galileo_Bun/Galileo_Bun_Overheat_Release_Cue.Galileo_Bun_Overheat_Release_Cue" (Variables|/Script/Engine.Actor|GetRootComponent (Game|GetInstigator)))
  (Components|Activation|Deactivate (Variables|Default|GetP_Overheat))
  (Game|Feedback|LocalPlayForceFeedback (Fortnite|GetFortPlayerControllerfromActor (Game|GetInstigator)) "/Game/Weapons/ForceFeedback/FF_GalileoBun_Overheated.FF_GalileoBun_Overheated" false "None"))

(event Custom|OverheatFX
  (params (Overheat%20Val :type Byte))
  (bind _returnvalue (Math|Conversions|ToInteger(Byte) Overheat%20Val))
  (bind _returnvalue_2 (Math|Conversions|ToFloat(Integer) _returnvalue))
  (bind _returnvalue_3 (/ _returnvalue_2 255.0))
  (bind _p_overheat (Variables|Default|GetP_Overheat))
  ;[1 1 1] Overheat FX
  (if (> _returnvalue_3 0.25)
    (Components|Activation|Activate _p_overheat true)
    (Effects|Components|ParticleSystem|SetEmitterEnable _p_overheat "MuzzleHeat" true)
    (Effects|Components|ParticleSystem|SetEmitterEnable _p_overheat "MuzzleOverheat" false)
    (Effects|Components|ParticleSystem|SetFloatParameter _p_overheat "MuzzleHeat" (Math|Float|Clamp(Float) (- (* _returnvalue_3 2.0) 1.0)))))

(event Custom|OverheatFXReplicated
  (params (Overheat%20Val :type Byte))
  (CallFunction|OverheatFX Overheat%20Val))

(event Overheating|EventOnStartHeating
  ;[1 1 1] Start Heating
  (|CallParentFunction)
  (bind _returnvalue (Utilities|Time|SetTimerbyEvent (delegate Custom|CustomEvent_0) (select (Fortnite|IsMobilePlatform) 0.1 0.03) true))
  (Variables|Default|SetOverheatFXTimer _returnvalue))

(event Custom|CustomEvent_1
  ;[1 1 1] Start Heating
  ;[1 0.03409 0] Replicaiton disabled for perf reasons
  (CallFunction|OverheatFXReplicated (Math|Conversions|ToByte(Integer) (Math|Float|Truncate (* 255 (/ (Overheat|GetCurrentOverheatValue) (Overheat|GetOverheatingMaxValue)))))))

(event Custom|CustomEvent_0
  ;[1 1 1] Start Heating
  (CallFunction|OverheatFX (Math|Conversions|ToByte(Integer) (Math|Float|Truncate (* 255 (/ (Overheat|GetCurrentOverheatValue) (Overheat|GetOverheatingMaxValue)))))))

(event Overheating|EventOnFinishedCooling
  ;[1 1 1] Finished Cooling
  (|CallParentFunction)
  (Components|Activation|Deactivate (Variables|Default|GetP_Overheat)))

(event Overheating|EventOnStartCooling
  ;[1 1 1] Start Cooling
  (|CallParentFunction)
  (Utilities|Time|ClearandInvalidateTimerbyHandle (Variables|Default|GetOverheatFXTimer))
  (Utilities|Time|ClearandInvalidateTimerbyHandle (Variables|Default|GetOverheatFXTimerReplicated)))

(event EventOnPlayWeaponFireFX
  (params (bPersistentFire :type Boolean) (bSecondaryFire :type Boolean))
  (bind _overheataudiolayer (Variables|Default|GetOverheatAudioLayer))
  (|CallParentFunction bPersistentFire bSecondaryFire)
  (bind _asfort_player_pawn (Utilities|Casting|CastToFortPlayerPawn (Game|GetInstigator)))
  (if (Pawn|IsLocallyControlled _asfort_player_pawn)
    (Audio|Components|Audio|Play _overheataudiolayer 0.0)
    (Audio|Components|Audio|SetVolumeMultiplier _overheataudiolayer (Math|Curves|GetFloatValue (Variables|Default|GetOverheatAudioFeedbackCurve) (/ (Overheat|GetCurrentOverheatValue) (Overheat|GetOverheatingMaxValue))) :!class /Script/Engine.AudioComponent)))

(fn ConstructionScript
  (|CallParentFunction))
