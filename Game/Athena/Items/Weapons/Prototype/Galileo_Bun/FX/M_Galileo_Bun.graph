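# Unlit, two-sided translucent particle material for Galileo_Bun weapon effects, drawing a procedural streaked glow.
# Panning noise distorts stretched, optionally rotated UVs feeding a sphere mask, shaping opacity with U falloff, particle alpha and depth fade.
# That mask plus panning detail texture indexes a curve-atlas ramp, then hue-shifted and HDR-boosted by time-of-day intensity.
# An optional second sphere brightens the core; particle color and dynamic desaturation tint the emissive.
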Material /Game/Athena/Items/Weapons/Prototype/Galileo_Bun/FX/M_Galileo_Bun
graph {
 properties {
  bCastRayTracedShadows=false
  bUsedWithNiagaraSprites=true
  bUsedWithParticleSprites=true
  blendMode=BLEND_Translucent
  pixelDepthOffsetMode=PDOM_Legacy
  shadingModel=MSM_Unlit
  twoSided=true
  usedShadingModels=Unlit
 }
 nodes [
  n1 = ParticleColor()
  n2 = ScalarParameter() {defaultValue=-0.05, group=Sphere_Ctrl, parameterName=Offset_U}
  n3 = ScalarParameter() {defaultValue=-0.07, group=Sphere_Ctrl, parameterName=Offset_V}
  n4 = TextureCoordinate()
  n5 = StaticSwitchParameter(
   True=MaterialFunctionCall(UVs=@n4, "Rotation Angle (0-1)"=ScalarParameter() {group="UV_Tweak(Shape_Ctrl)", parameterName="RotateUV(0-1)"}) {materialFunction=&CustomRotator}.Rotated Values
   False=@n4
  ) {group="UV_Tweak(Shape_Ctrl)", parameterName=RotateUV}
  n6 = Power(
   Base=OneMinus(None=ComponentMask(None=@n5) {r=true})
   Exp=ScalarParameter() {defaultValue=8, group="UV_Tweak(Shape_Ctrl)", parameterName=U_Stretch_Power}
  ) {constExponent=3}
  n7 = AppendVector(
   A=LinearInterpolate(B=ScalarParameter() {defaultValue=0.39, group="UV_Tweak(Shape_Ctrl)", parameterName=Stretch_Amount}, Alpha=OneMinus(None=@n6)) {constB=0.35}
   B=ComponentMask(None=@n5) {g=true}
  )
  n8 = AppendVector(
   A=ScalarParameter() {defaultValue=0.5, group=Distort, parameterName=Distortion_U}
   B=ScalarParameter() {defaultValue=3, group=Distort, parameterName=Distortion_V}
  )
  n9 = ScalarParameter() {defaultValue=-10, group=Distort, parameterName=Distortion_Speed_U}
  n10 = ScalarParameter() {group=Distort, parameterName=Distortion_Speed_V}
  n11 = TextureSampleParameter2D(UVs=Panner(Coordinate=Multiply(A=@n5, B=@n8) {constB=2}, Speed=AppendVector(A=@n9, B=@n10)) {speedX=-10}) {
   group=Distort
   parameterName=DistortionTexture
   samplerType=SAMPLERTYPE_LinearGrayscale
   texture=/Engine/Functions/Engine_MaterialFunctions02/ExampleContent/Textures/LowResBlurredNoise
  }
  n12 = Multiply(A=@n11.R, B=ScalarParameter() {defaultValue=0.12, group=Distort, parameterName=Distortion_Intensity}) {constB=0.1}
  n13 = AppendVector(
   A=ScalarParameter() {defaultValue=0.3, group=Sphere_Ctrl, parameterName=CenterX}
   B=ScalarParameter() {defaultValue=0.5, group=Sphere_Ctrl, parameterName=CenterY}
  )
  n14 = SphereMask(
   A=Add(A=AppendVector(A=@n2, B=@n3), B=Add(A=@n12, B=@n7))
   B=@n13
   Radius=ScalarParameter() {defaultValue=0.3, group=Sphere_Ctrl, parameterName=Radius_A}
   Hardness=ScalarParameter() {defaultValue=0.35, group=Sphere_Ctrl, parameterName=Hardness_A}
  ) {hardnessPercent=35}
  n15 = Saturate(None=@n14)
  n16 = OneMinus(None=ComponentMask(None=@n5) {r=true})
  n17 = Multiply(A=@n16, B=@n16)
  n18 = Multiply(A=@n1.A, B=Multiply(A=Power(Base=@n15) {constExponent=6}, B=@n17))
  n19 = DepthFade(Opacity=@n18) {fadeDistanceDefault=10}
  Add(A=ShadingPathSwitch(Default=@n19, Deferred=@n19, Mobile=@n18) {desc=Migrated from FeatureLevelSwitch}) {constB=0}
  n20 = ScalarParameter() {defaultValue=1, group=Detail, parameterName=Detail_U}
  n21 = ScalarParameter() {defaultValue=1, group=Detail, parameterName=Detail_V}
  n22 = Panner(
   Coordinate=Multiply(A=@n5, B=AppendVector(A=@n20, B=@n21))
   Speed=AppendVector(
    A=ScalarParameter() {defaultValue=-5, group=Detail, parameterName=Detail_Speed_U}
    B=ScalarParameter() {group=Detail, parameterName=Detail_Speed_V}
   )
  ) {speedX=-1}
  n23 = StaticComponentMaskParameter(None=TextureSampleParameter2D(UVs=Add(A=@n22, B=@n12)) {group=Detail, parameterName=DetailTexture, texture=/Game/Global/Textures/Default/Masks_and_Gradients/Cracks_02}.RGB) {defaultR=true, group=Detail, parameterName=DetailMask}
  n24 = ScalarParameter() {defaultValue=0.05, group=Detail, parameterName=Detail_Intensity}
  n25 = CurveAtlasRowParameter(CurveTime=Add(A=Multiply(A=@n23, B=@n24) {constB=0.05}, B=@n14)) {
   atlas=&AthenaCurveAtlas01
   curve=&/Curve_Galileo_Red
   defaultValue=150
   parameterName=ColorRamp
  }
  n26 = MaterialFunctionCall(Hue Shift Percentage=ScalarParameter() {parameterName=HueShift}, Texture=@n25) {materialFunction=&HueShift}
  n27 = VectorParameter() {parameterName=EmissiveTimeOfDay, defaultValue=<rgba 0.8 1 0.8 1>}
  n28 = Multiply(
   A=ScalarParameter() {defaultValue=1, parameterName=EmissiveIntensity}
   B=MaterialFunctionCall(Value_Morning=@n27.R, Value_Day=@n27.R, Value_Evening=@n27.R, Value_Night=@n27.R) {materialFunction=&Scalar_Time_of_Day}.Result
  )
  n29 = MaterialFunctionCall(
   LDR Input=Saturate(None=@n26.Result)
   HDR Tint=Multiply(A=@n26.Result, B=LinearInterpolate(B=@n28, Alpha=@n25.R)) {constB=3}
   Derive Luminance from Linear=StaticBool() {value=true}
  ) {materialFunction=&DeriveHDRfromLDR}
  n30 = Add(A=@n12, B=@n7)
  n31 = ScalarParameter() {group=Sphere_Ctrl, parameterName=Distort_Intensity_B}
  n32 = Multiply(
   A=SphereMask(
    A=LinearInterpolate(A=@n7, B=@n30, Alpha=@n31)
    B=@n13
    Radius=ScalarParameter() {defaultValue=0.08, group=Sphere_Ctrl, parameterName=Radius_B}
    Hardness=ScalarParameter() {group=Sphere_Ctrl, parameterName=Hardness_B}
   ) {hardnessPercent=20}
   B=ScalarParameter() {defaultValue=5, group=Sphere_Ctrl, parameterName=Radius_B_Bright}
  )
  n33 = Multiply(
   A=LinearInterpolate(
    A=Desaturation(None=@n29.RemappedRange)
    B=@n29.RemappedRange
    Alpha=DynamicParameter() {paramNames ["Desaturate(0-1)", StartTime, WPODuration, Param4]}."Desaturate(0-1)"
   )
   B=StaticSwitchParameter(True=Add(A=@n32), False=Constant() {r=1}) {defaultValue=true, group=Sphere_Ctrl, parameterName=UseSecondSphere}
  )
  MaterialFunctionCall(Emissive=Multiply(A=Multiply(A=@n33, B=@n1.RGB), B=@n17)) {materialFunction=&MF_FX_EmissiveAdjustments}
 ]
}
