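# Translucent unlit FX material for the Galileo weapon, with static switches for melee trail and projectile variants.
# Axis-aligned fresnel or trail UVs drive a curve-atlas ramp, hue-shifted into HDR emissive with optional flicker; mobile uses a three-color blend.
# A timed deflect sphere mask boosts glow and noise-based vertex wobble, blended between red and blue styles.
# Projectiles instead stretch vertices toward BulletOrigin, optionally Niagara-driven, with opacity fading by axis and camera distance.
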Material /Game/Athena/Items/Weapons/Prototype/Galileo_Bun/FX/M_Galileo_HDR_ColorRamp
graph {
 properties {
  bUsedWithBeamTrails=true
  bUsedWithMeshParticles=true
  bUsedWithNiagaraMeshParticles=true
  bUsedWithNiagaraRibbons=true
  blendMode=BLEND_Translucent
  pixelDepthOffsetMode=PDOM_Legacy
  shadingModel=MSM_Unlit
  usedShadingModels=Unlit
 }
 nodes [
  n1 = Multiply(A=Time(), B=ScalarParameter() {defaultValue=1, parameterName=FlickerSpeed})
  n2 = Sine(None=@n1) {period=0.075}
  n3 = Multiply(A=Multiply(A=Sine(None=@n1) {period=0.155}, B=Sine(None=@n1) {period=0.2}), B=Sine(None=@n1) {period=0.43})
  n4 = LinearInterpolate(Alpha=Saturate(None=Multiply(A=@n2, B=@n3))) {constA=0.25, constB=1.25}
  n5 = ScalarParameter() {defaultValue=0.99, group=Color, parameterName=HueColorShift}
  n6 = MaterialFunctionCall(Float2=MaterialFunctionCall(UVs=TextureCoordinate() {uTiling=0.05}, "Rotation Angle (0-1)"=ScalarParameter() {parameterName=Rotate_Angle}) {materialFunction=&CustomRotator}.Rotated Values) {materialFunction=&BreakOutFloat2Components}
  n7 = VertexNormalWS()
  n8 = TextureCoordinate() {uTiling=2, vTiling=4}
  n9 = ScalarParameter() {parameterName=Rotate_Angle}
  n10 = TextureSample(UVs=Panner(Coordinate=MaterialFunctionCall(UVs=@n8, "Rotation Angle (0-1)"=@n9) {materialFunction=&CustomRotator}.Rotated Values) {speedX=5, speedY=11}) {texture=/Game/Global/Textures/Default/Noise/3xPerlin_M}
  n11 = Divide(A=@n10.RGB, B=ScalarParameter() {defaultValue=2, group=Color, parameterName=NormalDivider}) {constB=1}
  n12 = MaterialFunctionCall(
   Axis=Transform(None=VectorParameter() {group=Color, parameterName=AxisAlignedFresnel, defaultValue=<rgba 0 0 1 1>}.RGB) {transformSourceType=TRANSFORMSOURCE_Local}
   Normal=LinearInterpolate(A=Add(A=@n7, B=@n11) {constB=0}, B=@n7, Alpha=ScalarParameter() {group=RedBlueSwitch, parameterName=RedBlueSwitch})
  ) {materialFunction=&AxisAlignedFresnel}
  n13 = Power(Base=@n12.Fresnel, Exp=ScalarParameter() {defaultValue=3, group=Color, parameterName=AxisAlignedFresnelExp})
  n14 = CurveAtlasRowParameter(CurveTime=StaticSwitchParameter(True=OneMinus(None=@n6.R), False=Saturate(None=@n13)) {group=MeleeTrail, parameterName=UseForMeleeTrail}) {
   atlas=&AthenaCurveAtlas01
   curve=&/Curve_Galileo_Red
   defaultValue=150
   group=Color
   parameterName=ColorRamp
  }
  n15 = AppendVector(A=AppendVector(A=@n14.R, B=@n14.R), B=@n14.R)
  n16 = VectorParameter() {group=Color, parameterName=Hue_Offset_Individual, defaultValue=<rgba 1 1 1 1>}
  n17 = MaterialFunctionCall(Hue Shift Percentage=@n5, Texture=StaticSwitchParameter(True=Multiply(A=@n15, B=@n16.RGB), False=@n14) {parameterName=Use_FadeByAxis}) {materialFunction=&HueShift}
  n18 = Saturate(None=@n17.Result)
  n19 = ScalarParameter() {defaultValue=200, group=Color, parameterName=EmissiveIntensity}
  n20 = ScalarParameter() {defaultValue=1000, group=Color, parameterName=EmissiveDistanceFade}
  n21 = VectorParameter() {group=Color, parameterName=EmissiveTimeOfDay, defaultValue=<rgba 0.8 1 0.8 1>}
  n22 = Multiply(
   A=LinearInterpolate(
    A=@n19
    B=Multiply(A=@n19) {constB=0.1}
    Alpha=MaterialFunctionCall(Distance=@n20) {materialFunction=&Near_Camera_Fade}.Result
   ) {constB=0}
   B=MaterialFunctionCall(Value_Morning=@n21.R, Value_Day=@n21.R, Value_Evening=@n21.R, Value_Night=@n21.R) {materialFunction=&Scalar_Time_of_Day}.Result
  )
  n23 = Multiply(A=@n22) {constB=6}
  n24 = Divide(
   A=Subtract(A=Time(), B=ScalarParameter() {defaultValue=1, desc=Time-stamp, group=ProjectileDeflect, parameterName=GameTime})
   B=ScalarParameter() {defaultValue=0.5, group=ProjectileDeflect, parameterName=ProjectileDeflectEffectDuration}
  )
  n25 = LinearInterpolate(
   B=SphereMask(
    A=WorldPosition().XYZ
    B=VectorParameter() {group=ProjectileDeflect, parameterName=ProjectileDeflectLocation}.RGB
    Radius=ScalarParameter() {defaultValue=120, group=ProjectileDeflect, parameterName=ProjectileDeflectSphereMaskRadius}
   ) {attenuationRadius=70, hardnessPercent=0}
   Alpha=OneMinus(None=Saturate(None=@n24))
  )
  n26 = Constant()
  n27 = QualitySwitch(Default=ShadingPathSwitch(Default=Saturate(None=@n25), Mobile=@n26) {desc=Migrated from FeatureLevelSwitch}, Low=@n26)
  n28 = Multiply(A=@n17.Result, B=LinearInterpolate(B=LinearInterpolate(A=@n22, B=@n23, Alpha=@n27) {constA=0.65}, Alpha=@n14.R)) {constB=3}
  n29 = StaticBool() {value=true}
  n30 = MaterialFunctionCall(
   Hue Shift Percentage=Add(A=ScalarParameter() {group=Color, parameterName=Moblie_HueColorShiftAdjustment}, B=@n5)
   Texture=MaterialFunctionCall(
    A=Constant3Vector() {constant=<rgba 10 0 0 0>}
    B=Constant3Vector() {constant=<rgba 2.5 0.01 0.02 0>}
    C=Constant3Vector() {constant=<rgba 6 0 0.05 0>}
    Alpha=@n12.Fresnel
   ) {materialFunction=&3ColorBlend}.Result
  ) {materialFunction=&HueShift}
  n31 = Multiply(
   A=ScalarParameter() {defaultValue=1.5, group=Color, parameterName=Moblie_EmissiveIntensity}
   B=ScalarParameter() {defaultValue=200, group=Color, parameterName=EmissiveIntensity}
  )
  n32 = ScalarParameter() {parameterName=Emissive_Offset}
  n33 = Subtract(B=@n32) {constA=-0.05, constB=2}
  n34 = Subtract(B=@n32) {constA=1.5, constB=2}
  n35 = TextureCoordinate()
  n36 = ScalarParameter() {parameterName=Rotate_Angle}
  n37 = OneMinus(None=ComponentMask(None=MaterialFunctionCall(UVs=@n35, "Rotation Angle (0-1)"=@n36) {materialFunction=&CustomRotator}.Rotated Values) {g=true})
  n38 = Multiply(A=@n37, B=@n37)
  n39 = StaticSwitchParameter(True=Saturate(None=LinearInterpolate(A=@n33, B=@n34, Alpha=@n38) {constA=-0.05, constB=1.5}), False=Constant() {r=1}) {parameterName=Use_LinearMask}
  n40 = ScalarParameter() {defaultValue=1, parameterName=EmissivePower}
  n41 = Multiply(
   A=ShadingPathSwitch(
    Default=MaterialFunctionCall(LDR Input=@n18, HDR Tint=@n28, Derive Luminance from Linear=@n29) {materialFunction=&DeriveHDRfromLDR}.RemappedRange
    Mobile=Multiply(A=@n30.Result, B=@n31)
   ) {desc=Migrated from FeatureLevelSwitch}
   B=StaticSwitchParameter(True=Constant() {r=1}, False=Power(Base=@n39, Exp=@n40)) {parameterName=Use_FadeByAxis}
  )
  MaterialFunctionCall(Emissive=LinearInterpolate(A=Multiply(A=@n4, B=@n41), B=@n41, Alpha=ScalarParameter() {defaultValue=1, parameterName=UnstableState})) {materialFunction=&MF_FX_EmissiveAdjustments}
  n42 = Subtract(A=Time() {period=5}, B=ScalarParameter() {group=Projectile, parameterName=StartTime})
  n43 = ScalarParameter() {defaultValue=0.2, group=Projectile, parameterName=WPODuration}
  n44 = MaterialFunctionCall!(In=Saturate(None=Divide(A=@n42, B=@n43))) {materialFunction=/Game/Effects/Fort_Effects/Materials/_Functions/MF_Smoothstep_EaseOut}
  n45 = VertexColor()
  n46 = ActorPositionWS()
  n47 = Subtract(A=VectorParameter() {group=Projectile, parameterName=BulletOrigin}.RGB, B=@n46)
  n48 = Subtract(A=@n46, B=WorldPosition().XYZ)
  Multiply(
   A=Multiply(A=ScalarParameter() {defaultValue=1, parameterName=WPOMult}, B=Saturate(None=@n44.Result))
   B=Multiply(A=@n45.R, B=Add(A=@n47, B=@n48))
  )
  n49 = VertexNormalWS()
  n50 = LinearInterpolate(B=ScalarParameter() {defaultValue=5, group=Wobble, parameterName=RedWobbleintensity}, Alpha=@n10.R) {constA=-0.5, constB=1.5}
  n51 = Multiply(A=@n10.RGB, B=LinearInterpolate(A=@n50, B=Multiply(A=@n50) {constB=8}, Alpha=@n27) {constA=0.65})
  n52 = AppendVector(A=ComponentMask(None=VertexNormalWS()) {g=true, r=true}, B=Constant())
  n53 = TextureCoordinate() {uTiling=0, vTiling=0.1}
  n54 = Multiply(A=Time()) {constB=4}
  n55 = ScalarParameter() {defaultValue=1.2, group=Wobble, parameterName=BlueWobbleintensity}
  n56 = Multiply(
   A=TextureSample(UVs=Panner(Coordinate=@n53, Time=@n54) {speedY=1}) {texture=&3xPerlin_M}.R
   B=LinearInterpolate(A=@n55, B=Multiply(A=@n55) {constB=15}, Alpha=@n27) {constA=0.65}
  )
  n57 = Multiply(
   A=LinearInterpolate(A=Multiply(A=@n49, B=@n51), B=Multiply(A=@n52, B=@n56), Alpha=ScalarParameter() {group=RedBlueSwitch, parameterName=RedBlueSwitch})
   B=VertexColor().R
  )
  n58 = VertexColor()
  n59 = ObjectPositionWS()
  n60 = Multiply(A=Time(), B=ScalarParameter() {defaultValue=40, group=Wobble, parameterName=TipWobbleSpeed}) {constB=40}
  n61 = TransformPosition(None=Subtract(A=WorldPosition().XYZ, B=Add(A=@n59, B=@n60))) {transformSourceType=TRANSFORMPOSSOURCE_World}
  n62 = TextureSample(UVs=ChannelMaskParameter(None=Divide(A=@n61) {constB=20}) {group=Wobble, maskChannel=Blue, parameterName=TipWobbleAxis, defaultValue=<rgba 0 0 1 0>}.R) {texture=&3xPerlin_M}
  n63 = MaterialFunctionCall(X=@n62.G, Y=@n62.G, Z=@n62.G) {materialFunction=&MakeFloat3}
  n64 = Multiply(
   A=LinearInterpolate(A=Multiply(A=@n63.Result) {constB=-0.7}, B=@n62.RGB, Alpha=@n62.RGB)
   B=ScalarParameter() {defaultValue=4, group=Wobble, parameterName=TipWobbleIntensity}
  ) {constB=4}
  n65 = Constant()
  n66 = ShadingPathSwitch(
   Default=LinearInterpolate(A=Multiply(A=@n58.G, B=@n64), Alpha=ScalarParameter() {group=RedBlueSwitch, parameterName=RedBlueSwitch}) {constB=0}
   Mobile=@n65
  ) {desc=Migrated from FeatureLevelSwitch}
  n67 = QualitySwitch(Default=@n66, Low=@n65, High=@n66, Medium=@n65)
  n68 = ScalarParameter() {parameterName=StretchPos_Offset}
  n69 = TextureCoordinate()
  n70 = ScalarParameter() {parameterName=Rotate_Angle}
  n71 = LinearInterpolate(
   A=Subtract(B=@n68) {constA=-2, constB=2}
   B=Subtract(B=@n68) {constB=2}
   Alpha=MaterialFunctionCall(Float2=MaterialFunctionCall(UVs=@n69, "Rotation Angle (0-1)"=@n70) {materialFunction=&CustomRotator}.Rotated Values) {materialFunction=&BreakOutFloat2Components}.G
  ) {constA=-2}
  n72 = DynamicParameter() {paramNames [BulletOrigin.X, BulletOrigin.Y, BulletOrigin.Z, Animation]}
  n73 = MaterialFunctionCall(R=@n72.BulletOrigin.X, G=@n72.BulletOrigin.X, B=@n72.BulletOrigin.X) {materialFunction=&AppendMany}
  n74 = VectorParameter() {group=Projectile, parameterName=BulletOrigin}
  n75 = StaticSwitchParameter(True=ParticlePositionWS(), False=ActorPositionWS()) {parameterName=Use with NDC}
  n76 = Multiply(
   A=Multiply(A=Clamp(None=@n71), B=ScalarParameter() {defaultValue=1, parameterName=Stretch_Amount}) {constB=1.1}
   B=Subtract(A=StaticSwitchParameter(True=@n73.RGB, False=@n74.RGB) {parameterName=Use with NDC}, B=@n75)
  )
  n77 = ScalarParameter() {defaultValue=1, parameterName=WPO_Strength}
  n78 = Time() {period=5}
  n79 = ScalarParameter() {group=Projectile, parameterName=StartTime}
  n80 = Saturate(None=Divide(A=Subtract(A=@n78, B=@n79), B=ScalarParameter() {defaultValue=0.2, group=Projectile, parameterName=WPODuration}))
  n81 = Saturate(None=StaticSwitchParameter(
    True=@n72.Animation
    False=MaterialFunctionCall(In=@n80) {materialFunction=&MF_Smoothstep_EaseOut}.Result
   ) {parameterName=Use with NDC})
  n82 = Multiply(A=Multiply(A=Multiply(A=@n76, B=@n77), B=@n45.R), B=@n81)
  n83 = WorldPosition()
  n84 = StaticSwitchParameter(True=Multiply(A=Subtract(A=@n75, B=@n83.XYZ), B=@n45.R), False=Constant()) {parameterName=Use_Shrink_to_ccenter}
  Reroute(None=StaticSwitchParameter(True=Add(A=@n82, B=@n84) {constB=0}, False=Add(A=@n57, B=@n67)) {group=Projectile, parameterName=UesForProjectile})
  ShadingPathSwitch(Default=@n27, Forward=@n26, Mobile=@n26)
  n85 = TransformPosition(None=WorldPosition().XYZ) {transformSourceType=TRANSFORMPOSSOURCE_World}
  n86 = Subtract(
   A=TransformPosition(None=ObjectPositionWS()) {transformSourceType=TRANSFORMPOSSOURCE_World}
   B=VectorParameter() {group=Active, parameterName=ActivationSphereMaskPivot, defaultValue=<rgba 0 0 45 1>}.RGB
  )
  n87 = LinearInterpolate(Alpha=ScalarParameter() {defaultValue=1, group=Active, parameterName=SphereMaskRadiusLerp}) {constB=150}
  n88 = ParticleColor()
  n89 = ScalarParameter() {defaultValue=2, group=Opacity, parameterName=OpacityEdgeFadeOutExp}
  n90 = Multiply(
   A=StaticSwitchParameter(True=SphereMask(A=@n85, B=@n86, Radius=@n87) {hardnessPercent=0}, False=Constant() {r=1}) {defaultValue=true, parameterName=Use_SphereMask}
   B=StaticSwitchParameter(True=Multiply(A=@n14.R, B=@n88.A), False=Power(Base=@n12.Fresnel, Exp=@n89)) {group=MeleeTrail, parameterName=UseForMeleeTrail}
  ) {constA=1}
  n91 = Multiply(A=@n90, B=Power(Base=@n39, Exp=ScalarParameter() {defaultValue=2, parameterName=OpacityPower}))
  n92 = Transform(None=VectorParameter() {group=Color, parameterName=ObjectAxis, defaultValue=<rgba 1 0 0 1>}.RGB) {transformSourceType=TRANSFORMSOURCE_Local}
  n93 = DotProduct(A=CameraVectorWS(), B=Multiply(A=Normalize(VectorInput=@n92)) {constB=-1})
  n94 = LinearInterpolate(A=@n91, B=@n90, Alpha=Clamp(None=LinearInterpolate(Alpha=@n93) {constA=-10}))
  n95 = Clamp(None=LinearInterpolate(
    A=ScalarParameter() {defaultValue=-10, parameterName=DisappearLerpA}
    B=ScalarParameter() {defaultValue=1, parameterName=DisappearLerpB}
    Alpha=@n93
   ) {constA=-10})
  n96 = StaticSwitchParameter(True=StaticSwitchParameter(True=Multiply(A=@n94, B=@n95), False=@n94) {parameterName=Use_Fade_at_side}, False=@n91) {parameterName=Use_FadeByAxis}
  n97 = VertexColor()
  n98 = LinearInterpolate(A=Power(Base=OneMinus(None=@n97.R)) {constExponent=0.5}, Alpha=@n81) {constA=1}
  n99 = ScalarParameter() {defaultValue=20, parameterName=Camera_Fade_distance}
  StaticSwitchParameter(
   True=Multiply(
    A=Multiply(A=@n96, B=@n98)
    B=MaterialFunctionCall(Distance=@n99) {materialFunction=&Near_Camera_Fade}.Result
   )
   False=@n96
  ) {group=Projectile, parameterName=UesForProjectile}
 ]
}
