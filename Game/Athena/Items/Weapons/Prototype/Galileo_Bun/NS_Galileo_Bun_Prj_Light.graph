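# Niagara system for the Galileo_Bun prototype weapon's projectile light: a single burst emitter drawing a red light, offset slightly from the system location, with no other visuals.
# Light color is scaled by time of day (dimmer at night), a near-camera fade, and a curve fading out over the system timeline.
# The same camera fade drives light radius.
# A red-to-green switch on User.ProjCollided is present but disabled.
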NiagaraSystem /Game/Athena/Items/Weapons/Prototype/Galileo_Bun/NS_Galileo_Bun_Prj_Light
dependencies {
 usedDataInterfaces [&NiagaraDataInterfaceCurve]
 usedDynamicInputs [&ClampFloat, &FN_LOD_Distance_Remap]
 usedModules [
  &self:SystemSpawnScript.SystemScriptSource.SystemScriptGraph.NiagaraNodeAssignment_0.SetVariables_#1
  &self:SystemSpawnScript.SystemScriptSource.SystemScriptGraph.NiagaraNodeAssignment_1.SetVariables_#2
  &SystemState
  &Timeline
 ]
}
emitters {PureLight {
  emitterUpdateScript [
   &EmitterState(Life Cycle Mode=::NewEnumerator1, Loop Behavior=::NewEnumerator1, Loop Duration Mode=::NewEnumerator1, UseLoopDelay=false)
   &SpawnBurst_Instantaneous(Use Spawn Probability=false, UseLoopCountLimit=false)
  ]
  particleSpawnScript [
   &InitializeParticle(Color=<rgba 10 0 0 1>, Lifetime=5, UsePositionOffset=false, Write Material Random=true)
   &SystemLocation(MaskSpawnGroup=false, Offset=<-30 0 0>)
  ]
  particleUpdateScript [
   &ParticleState(Kill Particles When Lifetime Has Elapsed=false, Let Infinitely Lived Particles Die When Emitter Deactivates=true)
   &ScaleColor(
    Scale Mode=::NewEnumerator1
    Scale RGBA=&MakeVector4_FromFloat(Float=&Multiply_Float(
      A=&Multiply_Float(
       A=&FNTimeOfDayScalar(Night Scale=0.4, Morning Scale=1, Evening Scale=1)
       B=$System.NearCamFadeLerp
      )
      B=&FloatFromCurve(
       FloatCurve=DataInterface() {propertyValues {Curve {keys [
           {arriveTangent=0.186, interpMode=RCIM_Cubic, leaveTangent=0.186, tangentMode=RCTM_User, time=0, value=1}
           {arriveTangent=-1.859, interpMode=RCIM_Cubic, leaveTangent=-1.859, tangentMode=RCTM_User, time=1, value=0}
          ]}}}
       CurveIndex=$System.Timeline.NormalizedTime
      )
     ))
    ScaleRGBA=true
   )
   /SetVariables(Particles.LightRadius=&Lerp_Float(A=0, B=100, Alpha=$System.NearCamFadeLerp)) {isSetParameters=true}
   &Color(Color=&LerpLinearColors(
     StartColor=<rgba 10 0 0 1>
     EndColor=<rgba 0 10 0 1>
     LerpFactor=&MakeCustomFloatFromBool(Bool=$User.ProjCollided, True Float=1, False Float=0)
    )) {enabled=false}
  ]
  rendererClasses [&NiagaraLightRendererProperties]
  renderers [{rendererClass=&NiagaraLightRendererProperties, rendererIndex=0}]
 }}
summary {
 systemName=NS_Galileo_Bun_Prj_Light
 emitters [{emitterName=PureLight, rendererClasses [&NiagaraLightRendererProperties]}]
 userVariables [User.ProjCollided {type=NiagaraBool}]
}
